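# Niagara swing trail effect for the Season 42 pickaxe.
# NE_MainTrail spawns an anim-trail ribbon between two system-set sockets, with three ribbon renderers and extra soft and offset tip color attributes, tinted by System.TOD.
# NE_Voxels scatters short-lived, spinning, shrinking meshes plus sprites in a cylinder, randomly colored from a gradient with fading glow.
# System.SwingGeneralFade scales both trail opacity and voxel spawn rate.
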NiagaraSystem /BRCosmetics/Weapons/FORT_Melee/Pickaxe_Season_42/FX/NS_Pickaxe_Season42_Trail
dependencies {
 usedDataInterfaces [&NiagaraDataInterfaceColorCurve, &NiagaraDataInterfaceCurve, &NiagaraDataInterfaceSocketReader]
 usedDynamicInputs [&MergeHelper_FloatSwitch]
 usedModules [
  &self:SystemSpawnScript.SystemScriptSource.SystemScriptGraph.NiagaraNodeAssignment_0.SetVariables_#1
  &self:SystemSpawnScript.SystemScriptSource.SystemScriptGraph.NiagaraNodeAssignment_2.SetVariables_#2
  &SystemState
  &self:SystemSpawnScript.SystemScriptSource.SystemScriptGraph.NiagaraNodeAssignment_1.SetVariables_#3
  &self:SystemSpawnScript.SystemScriptSource.SystemScriptGraph.NiagaraNodeAssignment_3.SetVariables_#4
 ]
}
emitters {
 NE_MainTrail {
  emitterSpawnScript [&FN_RibbonIDOffsetUtility()]
  emitterUpdateScript [
   &EmitterState(Life Cycle Mode=::NewEnumerator1, Loop Behavior=::NewEnumerator1, Loop Delay=0.05, Loop Duration=1, UseLoopDelay=true)
   &FN_SpawnAnimTrail()
  ]
  particleSpawnScript [
   &FN_InitializeAnimTrailBoneIndecies(First Bone or Socket Index:NiagaraInt32=$System.FirstSocket, Second Bone or Socket Index:NiagaraInt32=$System.SecondSocket)
   &FN_InitializeAnimTrail(
    Color=&MultiplyLinearColor(
     Linear Color=<rgba 1 1 1 0.4>
     Scale Factor=&MakeVector4FromVectorAndFloat("Vector (RGB)"=<1 1 1>, "Float (Alpha)"=$System.SwingGeneralFade)
    )
    Lifetime=0.15
   ) {module=FN_InitializeAnimTrail001}
  ]
  particleUpdateScript [
   &FN_UpdateAnimTrail()
   &ScaleColor(
    Scale Alpha=&FloatFromCurve(
     FloatCurve=DataInterface() {propertyValues {Curve {defaultValue=1, keys [{arriveTangent=-2.3795, interpMode=RCIM_Cubic, time=0, value=1}, <at 1 0 RCIM_Cubic>]}}}
     Scale Curve=&Multiply_Float(A=1, B=$System.SwingGeneralFade)
    )
    Scale RGB=&MakeVectorFromLinearColorRGB(Color=&MultiplyLinearColor(
      Linear Color=&ColorFromCurve(ColorCurve=DataInterface() {propertyValues {
         AlphaCurve {keys [<at 0 1 RCIM_Cubic>,<at 1 0 RCIM_Cubic>]}
         BlueCurve {keys [<at 0 0.3816 RCIM_Cubic>,<at 0.5019 0.1908>]}
         GreenCurve {keys [<at 0 1.4602 RCIM_Cubic>,<at 0.5019 0.7301>]}
         RedCurve {keys [<at 0 2 RCIM_Cubic>,<at 0.5019 1>]}
        }})
      Scale Factor=&MakeVector4_FromLinearColor(LinearColor=$System.TOD)
     ))
    ScaleA=true
    ScaleRGB=true
   )
   &DynamicMaterialParameters(
    Index 0 Param 3=0.1
    Module.Index0Param0Write=true
    Module.Index0Param1Write=true
    Module.Index0Param2Write=true
    Module.Index0Param3Write=true
    Module.Index1Param0Write=true
    Module.Index1Param1Write=true
    Module.Index1Param2Write=true
    Module.Index1Param3Write=true
    Write Parameter Index 1=true
   )
   /SetVariables(Particles.SoftTrailColor:LinearColor=&MultiplyLinearColor(
     Linear Color=&MultiplyLinearColor(
      Linear Color=&ColorFromCurve(ColorCurve=DataInterface() {propertyValues {
         AlphaCurve {keys [<at 0 1 RCIM_Cubic>,<at 1 0 RCIM_Cubic>]}
         BlueCurve {keys [<at 0 0.3816 RCIM_Cubic>,<at 0.6301 0.2841>,<at 0.7614 1>]}
         GreenCurve {keys [<at 0 1.4602 RCIM_Cubic>,<at 0.6301 0.7737>,<at 0.7614 0.7843>]}
         RedCurve {keys [<at 0 2 RCIM_Cubic>,<at 0.6301 1>,<at 0.7614 0.1815>]}
        }})
      Scale Factor=&MakeVector4_FromLinearColor(LinearColor=$System.TOD)
     )
     Scale Factor=&MakeVector4FromVectorAndFloat(
      "Vector (RGB)"=<1 1 1>
      "Float (Alpha)"=&FloatFromCurve(
       FloatCurve=DataInterface() {propertyValues {Curve {defaultValue=1, keys [{arriveTangent=-2.3795, interpMode=RCIM_Cubic, time=0, value=1}, <at 1 0 RCIM_Cubic>]}}}
       Scale Curve=&Multiply_Float(A=0.1, B=$System.SwingGeneralFade)
      )
     )
    )) {isSetParameters=true}
   /SetVariables(
    Particles.TipTrailColor:LinearColor=&MultiplyLinearColor(
     Linear Color=&MultiplyLinearColor(
      Linear Color=&ColorFromCurve(ColorCurve=DataInterface() {propertyValues {
         AlphaCurve {keys [<at 0 1 RCIM_Cubic>,<at 0.2964 0 RCIM_Cubic>]}
         BlueCurve {keys [<at 0 1.908 RCIM_Cubic>,<at 0.5019 0.1908>]}
         GreenCurve {keys [<at 0 7.3009 RCIM_Cubic>,<at 0.5019 0.7301>]}
         RedCurve {keys [<at 0 10 RCIM_Cubic>,<at 0.5019 1>]}
        }})
      Scale Factor=&MakeVector4_FromLinearColor(LinearColor=$System.TOD)
     )
     Scale Factor=&MakeVector4FromVectorAndFloat("Vector (RGB)"=<1 1 1>, "Float (Alpha)"=$System.SwingGeneralFade)
    )
    Particles.TipTrailPos:NiagaraPosition=&AddVectorToPosition(Position=$Particles.Position, Vector=&Multiply_VectorByFloat(Vector=$Particles.RibbonFacing, Float=30.5))
    Particles.TipTrailWidth=2
   ) {isSetParameters=true}
  ]
  rendererClasses [&NiagaraRibbonRendererProperties]
  renderers [
   {rendererClass=&NiagaraRibbonRendererProperties}
   {rendererClass=&NiagaraRibbonRendererProperties, rendererIndex=1}
   {rendererClass=&NiagaraRibbonRendererProperties, rendererIndex=2}
  ]
 }
 NE_Voxels {
  emitterUpdateScript [
   &EmitterState(Life Cycle Mode=::NewEnumerator1, Loop Behavior=::NewEnumerator1, Loop Delay=0.07, Loop Duration=1, UseLoopDelay=false)
   &SpawnRate(SpawnRate=&Multiply_Float(A=50, B=$System.SwingGeneralFade), Use Spawn Probability=true)
  ]
  particleSpawnScript [
   /SetVariables(Particles.CustomRandom:NiagaraFloat=&RandomRangeFloat(Minimum=0, Maximum=1)) {isSetParameters=true}
   &InitializeParticle(
    Color=&ColorFromCurve(
     ColorCurve=DataInterface() {propertyValues {
       AlphaCurve {keys []}
       BlueCurve {keys [
         {tangentMode=RCTM_User, time=0.1017, value=1}
         {arriveTangentWeight=0.0127, interpMode=RCIM_Cubic, leaveTangentWeight=0.2156, tangentMode=RCTM_User, time=0.1398, value=0.2542}
         {tangentMode=RCTM_User, time=0.7865, value=0.2542}
         {arriveTangentWeight=0.016, interpMode=RCIM_Cubic, leaveTangentWeight=0.016, tangentMode=RCTM_User, time=0.8344, value=1}
        ]}
       GreenCurve {keys [
         {tangentMode=RCTM_User, time=0.1017, value=0.7591}
         {
          arriveTangent=-0.757
          arriveTangentWeight=0.0159
          interpMode=RCIM_Cubic
          leaveTangent=-0.757
          leaveTangentWeight=0.2704
          tangentMode=RCTM_User
          time=0.1398
          value=0.7305
         }
         {tangentMode=RCTM_User, time=0.7865, value=0.7305}
         {arriveTangentWeight=0.016, interpMode=RCIM_Cubic, leaveTangentWeight=0.016, tangentMode=RCTM_User, time=0.8344, value=0.2045}
        ]}
       RedCurve {keys [
         {tangentMode=RCTM_User, time=0.1017, value=0.1585}
         {
          arriveTangent=1.0391
          arriveTangentWeight=0.0183
          interpMode=RCIM_Cubic
          leaveTangent=1.0391
          leaveTangentWeight=0.3109
          tangentMode=RCTM_User
          time=0.1398
          value=1
         }
         {tangentMode=RCTM_User, time=0.7865, value=1}
         {arriveTangentWeight=0.016, interpMode=RCIM_Cubic, leaveTangentWeight=0.016, tangentMode=RCTM_User, time=0.8344, value=0.9199}
        ]}
      }}
     CurveIndex=&RandomRangeFloat(Minimum=0, Maximum=1)
    )
    Lifetime Max=0.5
    Lifetime Min=0.3
    Lifetime Mode=::NewEnumerator1
    Mesh Scale Mode=::NewEnumerator3
    Mesh Uniform Scale=0.2
    Position Offset=<0 0 20>
    Sprite Size Mode=::NewEnumerator3
    Uniform Sprite Size=8
    UsePositionOffset=true
    Write Material Random=true
   )
   &InheritVelocity(Inherited Velocity Speed Limit=50, Speed Limit=true, UseSpeedThreshold=false)
   &ShapeLocation(
    Cylinder Height=63
    Cylinder Height Midpoint=0
    Cylinder Radius=5
    MaskSpawnGroup=false
    Override Seed=false
    Shape Primitive=::NewEnumerator1
    Surface Only=false
    UseOwnerScale=true
   )
  ]
  particleUpdateScript [
   &SolveForcesAndVelocity(Clamp Velocity=false, Limit Acceleration=false)
   &ParticleState()
   &ScaleMeshSize(Scale Factor=&VectorFromFloat(Value=&FloatFromCurve(FloatCurve=DataInterface() {propertyValues {Curve {keys [
          <at 0.05 1>
          {arriveTangent=-1.0526, interpMode=RCIM_Cubic, leaveTangent=-1.0526, time=0.5, value=1}
          {arriveTangent=-0.974, tangentMode=RCTM_User, time=1, value=0}
         ]}}})))
   &UpdateMeshOrientation(Rotation Rate=&RandomRangeFloat(Minimum=0.2, Maximum=0.3), Rotation Vector=<1 1 1>)
   &ScaleColor(
    Scale Alpha=&FloatFromCurve(FloatCurve=DataInterface() {propertyValues {Curve {keys [{interpMode=RCIM_Cubic, tangentMode=RCTM_Break, time=0.3, value=1}, <at 1 0>]}}})
    Scale RGB=&MakeVectorFromLinearColorRGB(Color=&ColorFromCurve(ColorCurve=DataInterface() {propertyValues {
        AlphaCurve {keys [<at 0 1>]}
        BlueCurve {keys [<at 0.0062 5>,<at 0.5636 1.5>,<at 0.9882 1.2>]}
        GreenCurve {keys [<at 0.0062 5>,<at 0.5636 1.5>,<at 0.9882 1.2>]}
        RedCurve {keys [<at 0.0062 5>,<at 0.5636 1.5>,<at 0.9882 1.2>]}
       }}))
    ScaleA=true
    ScaleRGB=true
   )
   /SetVariables(Particles.GlowColor=&MakeLinearColorFromVectorAndFloat(
     "Vector (RGB)"=&MakeVectorFromLinearColorRGB(Color=$Particles.Color)
     "Float (Alpha)"=&FloatFromCurve(FloatCurve=DataInterface() {propertyValues {Curve {keys [<at 0 1>,<at 1 0>]}}}, Scale Curve=0.1)
    )) {isSetParameters=true}
   &DynamicMaterialParameters(
    Index 0 Param 1=&FloatFromCurve(
     FloatCurve=DataInterface() {propertyValues {Curve {defaultValue=1, keys [{tangentMode=RCTM_User, time=0, value=-1}, {tangentMode=RCTM_User, time=1, value=0}]}}}
     CurveIndex=$Particles.CustomRandom
    )
    Module.Index0Float0Write=true
    Module.Index0Float1Write=true
    Module.Index0Float2Write=true
    Module.Index0Float3Write=true
   )
  ]
  rendererClasses [&NiagaraMeshRendererProperties, &NiagaraSpriteRendererProperties]
  renderers [
   {rendererClass=&NiagaraMeshRendererProperties, rendererIndex=0}
   {rendererClass=&NiagaraSpriteRendererProperties, rendererIndex=1}
  ]
 }
}
summary {
 systemName=NS_Pickaxe_Season42_Trail
 emitters [
  {emitterName=NE_MainTrail, rendererClasses [&NiagaraRibbonRendererProperties]}
  {emitterName=NE_Voxels, rendererClasses [&NiagaraMeshRendererProperties, &NiagaraSpriteRendererProperties]}
 ]
 userVariables [
  User.RibbonIDOffset {type=NiagaraInt32}
  User.SwingSectionIndex {type=NiagaraInt32}
 ]
}
