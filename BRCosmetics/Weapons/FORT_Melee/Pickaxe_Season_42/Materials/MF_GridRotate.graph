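# Material function that builds a procedural grid-line mask from UVs, with inputs for rotation, scale, cell counts, line thickness and smoothness.
# It rotates the UVs, scales them by center, then subtracts their tiled Fmod from the scaled UVs and takes the absolute value.
# A SmoothStep around the thickness threshold softens the result.
# The clamped output adds the R component to itself, leaving G unused.
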MaterialFunction /BRCosmetics/Weapons/FORT_Melee/Pickaxe_Season_42/Materials/MF_GridRotate
graph {nodes [
  n1 = FunctionInput() {
   bUsePreviewValueAsDefault=true
   description=Thickness of the grid lines.
   inputName=GridThickness
   inputType=FunctionInput_Scalar
   sortPriority=2
   previewValue=<0.02 0 0 1>
  }
  n2 = OneMinus(None=@n1)
  n3 = FunctionInput() {
   bUsePreviewValueAsDefault=true
   description=Gradient smoothness of the grid lines.
   inputName=GridSmooth
   inputType=FunctionInput_Scalar
   sortPriority=2
   previewValue=<0.2 0 0 1>
  }
  n4 = Divide(A=@n3) {constB=10}
  n5 = FunctionInput() {bUsePreviewValueAsDefault=true, description=Clockwise rotation value from 0 to 1, inputName=RotateGrid, inputType=FunctionInput_Scalar}
  n6 = Multiply(A=@n5, B=Constant() {r=-2})
  n7 = FunctionInput(Preview=TextureCoordinate()) {
   bUsePreviewValueAsDefault=true
   description="Position in UV space. Defaults to TexCoord[0]"
   inputName=UVs
   inputType=FunctionInput_Vector2
   sortPriority=-1
   previewValue=<0.5 0.5 0 1>
  }
  n8 = FunctionInput() {
   bUsePreviewValueAsDefault=true
   description=Horizontal scale of the entire grid.
   inputName=ScaleGridX
   inputType=FunctionInput_Scalar
   previewValue=<1 0 0 1>
  }
  n9 = FunctionInput() {
   bUsePreviewValueAsDefault=true
   description=Vertical scale of the entire grid.
   inputName=ScaleGridY
   inputType=FunctionInput_Scalar
   previewValue=<1 0 0 1>
  }
  n10 = MaterialFunctionCall(
   UVs=Rotator(Coordinate=@n7, Time=MaterialFunctionCall(Multiplier=@n6) {materialFunction=&Pi}.Result)
   Texture Scale=AppendVector(A=@n8, B=@n9)
  ) {materialFunction=&ScaleUVsByCenter}
  n11 = FunctionInput() {
   bUsePreviewValueAsDefault=true
   description=Amount of rows in the grid.
   inputName=GridAmountX
   inputType=FunctionInput_Scalar
   sortPriority=1
   previewValue=<8 0 0 1>
  }
  n12 = FunctionInput() {
   bUsePreviewValueAsDefault=true
   description=Amount of columns in the grid.
   inputName=GridAmountY
   inputType=FunctionInput_Scalar
   sortPriority=1
   previewValue=<8 0 0 1>
  }
  n13 = Multiply(A=@n10.UVs, B=AppendVector(A=@n11, B=@n12))
  n14 = Constant() {r=1}
  n15 = Abs(None=Subtract(A=@n10.UVs, B=Fmod(A=@n13, B=@n14)) {constB=0.5})
  n16 = MaterialFunctionCall(Float2=SmoothStep(Min=Subtract(A=@n2, B=@n4) {constA=0.8}, Max=Add(A=@n2, B=@n4) {constA=0.8}, Value=OneMinus(None=@n15))) {materialFunction=&BreakOutFloat2Components}
  FunctionOutput(None=Clamp(None=Add(A=@n16.R, B=@n16.R)))
 ]}
