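# Material function for the Season 42 pickaxe that blacks out base color and adds an animated orange-to-yellow grid glow to incoming material attributes.
# An AnimatedGrid function, projected from rotated, offset local XZ position with tunable tiling and speed, supplies colors plus grid lines.
# An FX mask texture scaled by GlowIntensity blends this over incoming emissive.
# Squares and lines masks are also derived from a Tech Art Mask input.
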MaterialFunction /BRCosmetics/Weapons/FORT_Melee/Pickaxe_Season_42/Materials/MF_Pickaxe_Season_42_Techart
graph {nodes [
  n1 = GetMaterialAttributes(None=FunctionInput() {inputName=Result, inputType=FunctionInput_MaterialAttributes}) {attributeGetTypes [#1]}
  FunctionOutput(None=SetMaterialAttributes!(MaterialAttributes=@n1.MaterialAttributes, EmissiveColor=NamedRerouteUsage(), BaseColor=Constant()) {attributeSetTypes [#1, #2]}) {outputName=SetMaterialAttributes}
  n2 = FunctionInput() {inputName=Tech Art Mask, inputType=FunctionInput_Scalar, sortPriority=10}
  NamedRerouteDeclaration(None=Power(Base=ComponentMask(None=@n2) {r=true}) {constExponent=8}) {name=SquaresMask, nodeColor=<rgba 0 1 0.2031 1>}
  NamedRerouteDeclaration(None=OneMinus(None=ComponentMask(None=@n2) {b=true})) {name=LinesMask, nodeColor=<rgba 1 0 0.1172 1>}
  NamedRerouteDeclaration(None=@n1.EmissiveColor) {name=EmissiveIN, nodeColor=<rgba 0 0.4375 1 1>}
  n3 = VectorParameter() {group=01_Season_42, parameterName=GlowTileOffset}
  n4 = ComponentMask(None=LocalPosition().XYZ) {b=true, r=true}
  n5 = Constant() {r=3.14}
  n6 = MaterialFunctionCall!(
   UV=Add(A=ComponentMask(None=@n3.RGB) {g=true, r=true}, B=Rotator(Coordinate=@n4, Time=@n5))
   Tiling=ScalarParameter() {defaultValue=0.5, group=01_Season_42, parameterName=GlowTiling}
   Tiling X
   Tiling Y
   Color 01=Constant4Vector() {constant=<rgba 1 0.3763 0 1>}.RGB
   Color 02=Constant4Vector() {constant=<rgba 1 1 0 1>}.RGB
   Speed=ScalarParameter() {defaultValue=0.1, group=01_Season_42, parameterName=GlowSpeed}
   Speed Random Amount
   Line Thickness
   Line Hardness
   Line Color
  ) {materialFunction=/Game/Global/MaterialFunctions/Cosmetics/AnimatedGrid/MF_AnimatedGrid}
  n7 = TextureSampleParameter2D() {
   group=01_Season_42
   parameterName=FX Mask 2
   samplerType=SAMPLERTYPE_LinearColor
   texture=/BRCosmetics/Weapons/FORT_Melee/Pickaxe_Season_42/Textures/T_Season_42_Pickaxe_FX2
  }
  n8 = ScalarParameter() {defaultValue=1, group=01_Season_42, parameterName=GlowIntensity}
  NamedRerouteDeclaration(None=LinearInterpolate(A=NamedRerouteUsage(), B=Add(A=@n6."Result (V3)", B=@n6.Gird Lines), Alpha=Multiply(A=@n7.R, B=@n8))) {name=EmissiveOUT, nodeColor=<rgba 0 0.0391 1 1>}
 ]}
